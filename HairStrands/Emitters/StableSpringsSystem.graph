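# Niagara system that simulates hair strand groom physics with a stable springs solver in a single GPU compute emitter.
# The emitter builds a velocity grid, sets up groom physics and spawns stable nodes as particles placed at groom locations.
# Each update stores state, transfers grid velocity, rewinds, attaches roots, advances with rod material off, updates the grid and reports positions.
# A mesh renderer is attached, and the reset-points spawn module is disabled.
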NiagaraSystem /HairStrands/Emitters/StableSpringsSystem
dependencies {
 usedDataInterfaces [
  /Script/HairStrandsCore.NiagaraDataInterfaceHairStrands
  /Script/Niagara.NiagaraDataInterfacePhysicsAsset
  /Script/HairStrandsCore.NiagaraDataInterfaceVelocityGrid
 ]
 usedDynamicInputs []
 usedModules [&SystemState]
}
emitters {StableSpringsEmitter {
  simTarget=GPUComputeSim
  emitterSpawnScript [&/SetupGridDatas(VelocityGrid=DataInterface() {propertyValues {GridSize=<25 25 25>}})]
  emitterUpdateScript [
   &EmitterState()
   &/SetupGroomPhysics()
   &/SpawnStableNodes()
  ]
  particleSpawnScript [
   /SetVariables(Particles.Lifetime=0, Particles.SpriteSize=<0.1 0.1>) {isSetParameters=true}
   &/ResetGroomPoints() {enabled=false}
   &/SetGroomLocation()
  ]
  particleUpdateScript [
   &/StoreGroomState()
   &/TransferGridVelocity()
   &/RewindStableState()
   &/AttachGroomRoots()
   &/AdvanceGroomState(Rod Material=false)
   &/UpdateGridVelocity()
   &/ReportGroomPosition()
  ]
  rendererClasses [&NiagaraMeshRendererProperties]
  renderers [{rendererClass=&NiagaraMeshRendererProperties, rendererIndex=0}]
 }}
summary {
 systemName=StableSpringsSystem
 emitters [{emitterName=StableSpringsEmitter, simTarget=GPUComputeSim, rendererClasses [&NiagaraMeshRendererProperties]}]
 userVariables []
}
